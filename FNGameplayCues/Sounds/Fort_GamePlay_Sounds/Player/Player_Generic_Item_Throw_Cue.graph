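# Sound cue for the generic player item throw sound used by gameplay cues.
# A single wave player feeds the Player_Generic_Item_Throw wave straight to the output, with no randomization or modulation.
# It uses the gadget-throw attenuation and gadgets sound class, with reduced volume and a limited audible range.
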SoundCue /FNGameplayCues/Sounds/Fort_GamePlay_Sounds/Player/Player_Generic_Item_Throw_Cue
root=WavePlayer_0
nodes [{id=WavePlayer_0, properties {soundWaveAssetPtr=&/Player_Generic_Item_Throw}}]
properties {
 attenuationSettings=/Game/Sounds/Attenuation/Fort_GadgetThrow_Att
 maxDistance=1536
 soundClassObject=/Game/Sounds/FortSoundClassesAndModes/Fort_Gadgets_StW
 volumeMultiplier=0.8
}
